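# Keybind prompt widget on UIKitModularButton showing an input action icon, a hold-progress image and a press-and-hold icon.
# UpdatePrompt sets their visibility and progress shape from input type, whether the action is held and whether an icon exists.
# Hold progress drives a dynamic material's Percentage parameter and a dispatcher.
# Button events request block visual transitions and play hover, press and hold sounds, deferring enable, disable and optional lock changes one tick.
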WidgetBlueprint /FortUILibrary/Systems/ModularButton/WBP_UIKit_Keybind_Holdable
graphs [
 CallOnReleased {nodes [CallOnReleased!() -> (then, Button:WBP UIKit Keybind Holdable Object Reference) {is=title K2Node_FunctionEntry}]}
 ClearPrompt {nodes [
   ClearPrompt() {is=title K2Node_FunctionEntry}
   > SetVisibility(
    self=[
     Get HoldProgressBar() -> (HoldProgressBar:Image Object Reference)
     Get Prompt() -> (Prompt:Grid Panel Object Reference)
     Get HoldIcon() -> (HoldIcon:Image Object Reference)
    ]
    InVisibility=Collapsed
   )
   > SetHoldPercentage!(execute, self, HoldPercentage:double=0) -> (then)
  ]}
 EventGraph {nodes [
   "On Input Icon Updated (InputActionWidget)"() {is=title K2Node_ComponentBoundEvent}
   > SetInputIconDarkMode!(execute, self, UseDark:bool=Get UseDarkInputIcon() -> (UseDarkInputIcon:bool)) -> (then)
   n1 = Event PreConstruct()
   > UpdateInputPrompt()
   > Branch(Condition=@n1.IsDesignTime)
   > n2 = GetDynamicMaterial(self=Get HoldProgressBar())
   > Set InputActionProgressMaterial(InputActionProgressMaterial:Material Instance Dynamic Object Reference=@n2) -> (Output_Get:Material Instance Dynamic Object Reference)
   > SetHoldPercentage(HoldPercentage=0.4)
   Event Construct()
   > SetPrompt()
   > SetBlocksTiming!(
    execute
    self:UIKit Modular Button Object Reference
    Timing:"UIKit Block Timing Structure (by ref)"=Make UIKit Block Timing!(
     Hovering:float=0
     Unhovering:float=0
     Focusing:float=0.3
     Unfocusing:float=0.15
     Pressing:float=0.1
     Releasing:float=0.1
     Disabling:float=0.1
     Enabling:float=0.1
     Selecting:float=0.2
     DeselectingIdle:float=0.2
     DeselectingFocused:float=0.2
    ) -> (UIKitBlockTiming:UIKit Block Timing Structure) {is=title K2Node_MakeStruct}
   ) -> (then)
   > ResetButtonState()
   n3 = Event OnActionProgress!() -> (OutputDelegate, then, HeldPercent:float) {is=title K2Node_Event}
   > SetHoldPercentage(HoldPercentage=@n3.HeldPercent)
   > Call On Hold Action Progress!(execute:Exec, self:WBP UIKit Keybind Holdable Object Reference, HoldPercentage:double=@n3.HeldPercent) -> (then:Exec) {is=title K2Node_CallDelegate}
   Event On Input Action Triggered() {is=title K2Node_Event}
   > PlaySoundPressed()
   n4 = Event On Locked Changed!() -> (OutputDelegate, then, bIsLocked:bool) {is=title K2Node_Event}
   > Branch(Condition=Get UseDisabledVisualsOnLocked() -> (UseDisabledVisualsOnLocked:bool))
   > n5 = Set LastEnabledEvent(LastEnabledEvent:bool=NOT Boolean(A=@n4.bIsLocked)) -> (Output_Get:bool)
   DelayUntilNextTick!(execute=[@n9.then, @n8.then, @n5.then]) -> (then)
   > SetButtonIsEnabled()
   Event On Hovered()
   > ShowPromptFallbackAction!(execute, self, Show:bool=true) -> (then)
   > RequestTransitionOnBlocks!(execute, self:UIKit Modular Button Object Reference, State:EUIKitBlockVisualState Enum=Focused) -> (then)
   > PlaySoundHovered()
   Event On Unhovered()
   > ShowPromptFallbackAction(Show=false)
   > RequestTransitionOnBlocks(State=Unfocused)
   Event On Pressed() {is=title K2Node_Event}
   > RequestTransitionOnBlocks(State=Pressed)
   > n6 = Branch(Condition=IsPromptHoldable!() -> (Holdable:bool))
   > PlaySoundHoldPressed()
   PlaySoundPressed(execute=@n6.else)
   Event On Released() {is=title K2Node_Event}
   > Call Call on Released!(execute:Exec, self:WBP UIKit Keybind Holdable Object Reference, Button:WBP UIKit Keybind Holdable Object Reference=Self-Reference()) -> (then:Exec) {is=title K2Node_CallDelegate}
   > RequestTransitionOnBlocks(State=Released)
   n7 = Get CurrentHoldProgress() -> (CurrentHoldProgress:double)
   > Branch(Condition=AND Boolean(A=IsPromptHoldable(), B="float < float"(A=@n7, B=1)))
   > PlaySoundHoldReleased()
   Event On Disabled() {is=title K2Node_Event}
   > n8 = Set LastEnabledEvent(LastEnabledEvent=false)
   Event On Enabled() {is=title K2Node_Event}
   > n9 = Set LastEnabledEvent(LastEnabledEvent=true)
   n10 = Event On Input Method Changed!() -> (OutputDelegate, then, CurrentInputType:ECommonInputType Enum) {is=title K2Node_Event}
   > n11 = Set CurrentInputType(CurrentInputType:ECommonInputType Enum=@n10.CurrentInputType) -> (Output_Get:ECommonInputType Enum)
   > UpdatePrompt!(execute, self, InputType:ECommonInputType Enum=@n11) -> (then)
   > FadeOutSound()
   SetEnhancedInputAction!(
    execute
    self:Common Action Widget Object Reference=Get InputActionWidget() -> (InputActionWidget:Common Action Widget Object Reference)
    InInputAction:Input Action Object Reference=Get TriggeringEnhancedInputAction() -> (TriggeringEnhancedInputAction:Input Action Object Reference)
   ) -> (then)
  ]}
 FadeOutSound {nodes [
   FadeOutSound() {is=title K2Node_FunctionEntry}
   > n1 = Sequence()
   > n2 = Get!(execute:Exec) -> (then:Exec, else:Exec, PressedSoundAudioComponent:Audio Component Object Reference)
   > FadeOut!(
    execute
    self:Audio Component Object Reference=@n2
    FadeOutDuration:float=Get FadeOutTime() -> (FadeOutTime:double)
    FadeVolumeLevel:float=0
    FadeCurve:EAudioFaderCurve Enum=Linear
   ) -> (then)
   n3 = Get!(execute=@n1.then_1) -> (then, else, HoldSoundAudioComponent:Audio Component Object Reference)
   > FadeOut(self=@n3, FadeOutDuration=Get FadeOutTime(), FadeVolumeLevel=0, FadeCurve=Linear)
  ]}
 IsPromptHoldable {nodes [
   IsPromptHoldable() {is=title K2Node_FunctionEntry}
   n1 = Get InputActionWidget()
   > Return Node!(
    execute
    Holdable:bool=OR Boolean(A=Get bRequiresHold() -> (bRequiresHold:bool), B=IsHeldAction!(self:Common Action Widget Object Reference=@n1) -> (ReturnValue:bool))
   )
  ]}
 IsPromptIcon {nodes [
   IsPromptIcon() {is=title K2Node_FunctionEntry}
   n1 = GetIcon!(self:Common Action Widget Object Reference=Get InputActionWidget()) -> (ReturnValue:Slate Brush Structure)
   > Return Node!(
    execute
    Valid:bool=IsValid(Object=Break Slate Brush!(SlateBrush:"Slate Brush Structure (by ref)"=@n1) -> (
      TintColor:Slate Color Structure
      DrawAs:ESlateBrushDrawType Enum
      Tiling:ESlateBrushTileType Enum
      Mirroring:ESlateBrushMirrorType Enum
      ImageSize:Vector 2D Structure
      Margin:Margin Structure
      ResourceObject:Object Reference
      OutlineSettings:Slate Brush Outline Settings Structure
     ) {is=title K2Node_BreakStruct}.ResourceObject)
   )
  ]}
 OnHoldActionProgress {nodes [OnHoldActionProgress!() -> (then, HoldPercentage:double=~) {is=title K2Node_FunctionEntry}]}
 PlaySoundHoldPressed {nodes [
   PlaySoundHoldPressed() {is=title K2Node_FunctionEntry}
   > n1 = Get!(execute) -> (then, else, HoldCompletedSound:Sound Base Object Reference)
   > SpawnSound2D!(
    execute
    Sound:Sound Base Object Reference=@n1
    VolumeMultiplier:float=1
    PitchMultiplier:float=1
    StartTime:float=0
    ConcurrencySettings:Sound Concurrency Object Reference
    bPersistAcrossLevelTransition:bool=true
    bAutoDestroy:bool=true
   ) -> (then, ReturnValue:Audio Component Object Reference)
   > FadeOutSound() {is=title K2Node_CallFunction}
  ]}
 PlaySoundHoldReleased {nodes [
   PlaySoundHoldReleased() {is=title K2Node_FunctionEntry}
   > n1 = Get!(execute) -> (then, else, HoldReleasedSound:Sound Base Object Reference)
   > n2 = SpawnSound2D(Sound=@n1, VolumeMultiplier=1, PitchMultiplier=1, StartTime=0, bPersistAcrossLevelTransition=false, bAutoDestroy=true)
   FadeOutSound(execute=[@n2.then, @n1.else])
  ]}
 PlaySoundHovered {nodes [
   PlaySoundHovered() {is=title K2Node_FunctionEntry}
   > n1 = Get!(execute) -> (then, else, HoveredSound:Sound Base Object Reference)
   > SpawnSound2D(Sound=@n1, VolumeMultiplier=1, PitchMultiplier=1, StartTime=0, bPersistAcrossLevelTransition=false, bAutoDestroy=true)
  ]}
 PlaySoundPressed {nodes [
   PlaySoundPressed() {is=title K2Node_FunctionEntry}
   > n1 = Get!(execute) -> (then, else, PressedSound:Sound Base Object Reference)
   > SpawnSound2D(Sound=@n1, VolumeMultiplier=1, PitchMultiplier=1, StartTime=0, bPersistAcrossLevelTransition=false, bAutoDestroy=true)
  ]}
 ResetButtonState {nodes [
   ResetButtonState() {is=title K2Node_FunctionEntry}
   > StopAllAnimations(self:User Widget Object Reference)
   > RequestInstantTransitionOnBlocks!(
    execute
    self:UIKit Modular Button Object Reference
    State:EUIKitBlockInstantTransitionState Enum=Select(
     Option 0:EUIKitBlockInstantTransitionState Enum=InstantDisabled
     Option 1:EUIKitBlockInstantTransitionState Enum=InstantEnabled
     Index=GetIsEnabled!(self:Widget Object Reference) -> (ReturnValue:bool)
    ) -> (ReturnValue:EUIKitBlockInstantTransitionState Enum)
   ) -> (then)
  ]}
 SetButtonIsEnabled {nodes [
   SetButtonIsEnabled() {is=title K2Node_FunctionEntry}
   > n1 = Branch(Condition=Get LastEnabledEvent() -> (LastEnabledEvent:bool))
   > n2 = Branch(Condition=IsHovered())
   > n3 = RequestTransitionOnBlocks(State=Enabled)
   SetPrompt(execute=[@n3.then, @n4.then])
   n4 = RequestInstantTransitionOnBlocks(execute=@n2.else, State=InstantEnabled)
   RequestTransitionOnBlocks(execute=@n1.else, State=Disabled)
   > SetVisibility(self=Get Prompt(), InVisibility=Collapsed)
  ]}
 SetHoldPercentage {nodes [
   n1 = SetHoldPercentage!() -> (then, HoldPercentage:double=~) {is=title K2Node_FunctionEntry}
   > n2 = Set CurrentHoldProgress(CurrentHoldProgress:double=@n1) -> (Output_Get:double)
   > SetScalarParameterValue(
    self=Get InputActionProgressMaterial() -> (InputActionProgressMaterial:Material Instance Dynamic Object Reference)
    ParameterName=Percentage
    Value=@n2
   )
  ]}
 SetInputIconDarkMode {nodes [
   n1 = SetInputIconDarkMode!() -> (then, UseDark:bool=~) {is=title K2Node_FunctionEntry}
   > Set UseDarkInputIcon(UseDarkInputIcon:bool=@n1) -> (Output_Get:bool)
   > n2 = GetIconDynamicMaterial!(execute, self:Common Action Widget Object Reference=Get InputActionWidget()) -> (then, ReturnValue:Material Instance Dynamic Object Reference)
   > n3 = Set CurrentDynamicPromptIcon(CurrentDynamicPromptIcon:Material Instance Dynamic Object Reference=@n2) -> (Output_Get:Material Instance Dynamic Object Reference)
   > SetScalarParameterValue(self=@n3, ParameterName=Use Dark, Value=Select(Option 0:float=0, Option 1:float=1, Index=Get UseDarkInputIcon()) -> (ReturnValue:float))
   > SetScalarParameterValue(
    self=Get CurrentDynamicPromptIcon() -> (CurrentDynamicPromptIcon:Material Instance Dynamic Object Reference)
    ParameterName=Outline Opacity
    Value=Select(Option 0=0.5, Option 1=0, Index=Get UseDarkInputIcon())
   )
  ]}
 SetPrompt {nodes [
   SetPrompt() {is=title K2Node_FunctionEntry}
   n1 = Get CommonInputSubsystem!() -> (ReturnValue:Common Input Object Reference) {is=title K2Node_GetBlueprintContextEx}
   > Branch(Condition=IsValid(Object=@n1))
   > n2 = GetDynamicMaterial(self=Get HoldProgressBar())
   > Set InputActionProgressMaterial(InputActionProgressMaterial=@n2)
   > n3 = Set CurrentInputType(CurrentInputType=GetCurrentInputType(self=@n1))
   > UpdatePrompt(InputType=@n3)
  ]}
 ShowPromptFallbackAction {nodes [
   n1 = ShowPromptFallbackAction!() -> (then, Show:bool=~) {is=title K2Node_FunctionEntry}
   n2 = Get CurrentInputType() -> (CurrentInputType:ECommonInputType Enum)
   > Branch(Condition=AND Boolean(A="Equal (Enum)"(A=@n2, B=Gamepad), B=Get bShouldUseFallbackDefaultInputAction() -> (bShouldUseFallbackDefaultInputAction:bool)))
   > SetVisibility(
    self=Get Prompt()
    InVisibility=Select(Option 0:ESlateVisibility Enum=Collapsed, Option 1:ESlateVisibility Enum=SelfHitTestInvisible, Index=@n1) -> (ReturnValue:ESlateVisibility Enum)
   )
  ]}
 UpdateInputPrompt {nodes [
   UpdateInputPrompt() {is=title K2Node_FunctionEntry}
   n1 = Get Prompt()
   > SetGridSlotPadding!(execute, Widget:Widget Object Reference=@n1, Padding:Margin Structure=Get PromptPadding() -> (PromptPadding:Margin Structure)) -> (then)
   > SetGridSlotNudge!(execute, Widget:Widget Object Reference=@n1, Nudge:Vector 2D Structure=Get PromptNudge() -> (PromptNudge:Vector 2D Structure)) -> (then)
   > SetGridSlotAlignment!(
    execute
    Widget:Widget Object Reference=@n1
    Horizontal:EHorizontalAlignment Enum=Get PromptHorizontalAlignment() -> (PromptHorizontalAlignment:EHorizontalAlignment Enum=HAlign_Fill)
    Vertical:EVerticalAlignment Enum=Get PromptVerticalAlignment() -> (PromptVerticalAlignment:EVerticalAlignment Enum=VAlign_Fill)
   ) -> (then)
  ]}
 UpdatePrompt {nodes [
   n1 = UpdatePrompt!() -> (then, InputType:ECommonInputType Enum) {is=title K2Node_FunctionEntry}
   > ClearPrompt() {is=title K2Node_CallFunction}
   > Set IsPromptHoldable(IsPromptHoldable:bool=IsPromptHoldable() {is=title K2Node_CallFunction}) -> (Output_Get:bool)
   > Set IsPromptIcon(IsPromptIcon:bool=IsPromptIcon!() -> (Valid:bool) {is=title K2Node_CallFunction}) -> (Output_Get:bool)
   n2 = Get bHideInputActionWithKeyboard() -> (bHideInputActionWithKeyboard:bool)
   n3 = Get IsPromptIcon() -> (IsPromptIcon:bool)
   n4 = Get IsPromptHoldable() -> (IsPromptHoldable:bool)
   n5 = Select(Option 0=Collapsed, Option 1=SelfHitTestInvisible, Index=AND Boolean(A=OR Boolean(A=@n3, B=@n4), B=NOT Boolean(A=@n2)))
   n6 = Select(Option 0=Collapsed, Option 1=SelfHitTestInvisible, Index=Get IsPromptHoldable())
   n7 = Select(Option 0=Collapsed, Option 1=SelfHitTestInvisible, Index=Get IsPromptIcon())
   > SetVisibility(
    self=Get Prompt()
    InVisibility=Select(
     Option 0=Collapsed
     Option 1=Select!(
      MouseAndKeyboard:ESlateVisibility Enum=@n5
      Gamepad:ESlateVisibility Enum=@n7
      Touch:ESlateVisibility Enum=@n6
      Count:ESlateVisibility Enum=Collapsed
      Index:ECommonInputType Enum=@n1
     ) -> (ReturnValue)
     Index=GetIsEnabled()
    )
   )
   > Branch(Condition=Get IsPromptHoldable())
   n8 = Get IsPromptIcon()
   > SetScalarParameterValue(
    self=Get InputActionProgressMaterial()
    ParameterName=ShapeSwitch
    Value=Select(
     MouseAndKeyboard:float=Select!(Option 0:float=0, Option 1:float=1, Index:bool=@n8) -> (ReturnValue)
     Gamepad:float=0
     Touch:float=0
     Count:float=0
     Index=@n1
    ) -> (ReturnValue:float)
   )
   > SetVisibility(self=Get HoldProgressBar(), InVisibility=SelfHitTestInvisible)
   n9 = Get IsPromptIcon()
   > SetVisibility(
    self=Get HoldIcon()
    InVisibility=Select!(
     MouseAndKeyboard:ESlateVisibility Enum=Select!(Option 0:ESlateVisibility Enum=SelfHitTestInvisible, Option 1:ESlateVisibility Enum=Collapsed, Index:bool=@n9) -> (ReturnValue)
     Gamepad:ESlateVisibility Enum=Collapsed
     Touch:ESlateVisibility Enum=SelfHitTestInvisible
     Count:ESlateVisibility Enum=Collapsed
     Index:ECommonInputType Enum=@n1
    ) -> (ReturnValue:ESlateVisibility Enum)
   )
  ]}
]
variables [
 CallOnReleased {}
 CurrentDynamicPromptIcon {type=MaterialInstanceDynamic}
 CurrentHoldProgress {type=number, default=0.4}
 CurrentInputType {type=ECommonInputType, default=Count}
 FadeOutTime {type=number}
 HoldCompletedSound {
  type=SoundBase
  default=/Game/Sounds/Fort_UI_Sounds/FrontEnd/DefaultSounds/HoldableButtons/Holdable_Regular/MSS_UI_Button_Holdable_Regular_Complete
 }
 HoldReleasedSound {
  type=SoundBase
  default=/Game/Sounds/Fort_UI_Sounds/FrontEnd/DefaultSounds/HoldableButtons/Holdable_Regular/MSS_UI_Button_Holdable_Regular_Release
 }
 HoldSoundAudioComponent {type=AudioComponent, default=null}
 HoveredSound {type=SoundBase, default=/Game/Sounds/Fort_UI_Sounds/FrontEnd/DefaultSounds/PressedButtons/MSS_UI_Button_Hover_Default}
 InputActionProgressMaterial {type=MaterialInstanceDynamic}
 LastEnabledEvent {type=boolean}
 OnHoldActionProgress {}
 PressedSound {type=SoundBase, default=/Game/Sounds/Fort_UI_Sounds/FrontEnd/DefaultSounds/PressedButtons/MSS_UI_Button_Pressed_Default}
 PressedSoundAudioComponent {type=AudioComponent, default=null}
 PromptHorizontalAlignment {type=EHorizontalAlignment, default=HAlign_Fill}
 PromptNudge {type=Vector2D, default=<0 0>}
 PromptPadding {type=Margin, default=<ltrb 0 0 0 0>}
 PromptVerticalAlignment {type=EVerticalAlignment, default=VAlign_Fill}
 UseDarkInputIcon {type=boolean}
 UseDisabledVisualsOnLocked {type=boolean}
]
widgets {
 info {inheritedWidgetCount=1, parentClass=/Script/FortniteUI.UIKitModularButton, rootWidgetClass=&GridPanel, widgetCount=5}
 widgets [
  GridPanel_26 {widgetClassPath=&GridPanel, widget=GridPanel_26, properties {columnFill [1], rowFill [1]}}
  Prompt {
   widgetClassPath=&GridPanel
   bIsVariable=true
   parent=GridPanel_26
   widget=Prompt
   properties {columnFill [1], rowFill [1]}
   slot {class=GridSlot, properties {horizontalAlignment=HAlign_Left, verticalAlignment=VAlign_Bottom}}
  }
  HoldProgressBar {
   widgetClassPath=&Image
   bIsVariable=true
   parent=Prompt
   widget=HoldProgressBar
   properties {
    visibility=SelfHitTestInvisible
    brush {
     bIsDynamicallyLoaded=false
     drawAs=Image
     imageType=NoImage
     mirroring=NoMirror
     resourceName=None
     resourceObject=/FortUILibrary/UIKit/Buttons/MaterialInstances/CTAButtons/MI_UIKit_InputActionProgress
     tiling=NoTile
     imageSize=<48 48>
     margin=<ltrb 0 0 0 0>
     outlineSettings {
      bUseBrushTransparency=false
      roundingType=HalfHeightRadius
      width=0
      color {colorUseRule=UseColor_Specified, specifiedColor=<rgba 0 0 0 0>}
      cornerRadii=<0 0 0 0>
     }
     tintColor {colorUseRule=UseColor_Specified, specifiedColor=<rgba 1 1 1 1>}
     uVRegion {bIsValid=false, max=<0 0>, min=<0 0>}
    }
   }
   slot {class=GridSlot, properties {horizontalAlignment=HAlign_Center, layer=10, verticalAlignment=VAlign_Center}}
  }
  HoldIcon {
   widgetClassPath=&Image
   bIsVariable=true
   parent=Prompt
   widget=HoldIcon
   properties {
    visibility=SelfHitTestInvisible
    brush {
     bIsDynamicallyLoaded=false
     drawAs=Image
     imageType=NoImage
     mirroring=NoMirror
     resourceName=None
     resourceObject=/FortUILibrary/Systems/ModularButton/MaterialInstances/MI_MousePressAndHold
     tiling=NoTile
     imageSize=<28 28>
     margin=<ltrb 0 0 0 0>
     outlineSettings {
      bUseBrushTransparency=false
      roundingType=HalfHeightRadius
      width=0
      color {colorUseRule=UseColor_Specified, specifiedColor=<rgba 0 0 0 0>}
      cornerRadii=<0 0 0 0>
     }
     tintColor {colorUseRule=UseColor_Specified, specifiedColor=<rgba 1 1 1 1>}
     uVRegion {bIsValid=false, max=<0 0>, min=<0 0>}
    }
   }
   slot {class=GridSlot, properties {horizontalAlignment=HAlign_Center, layer=30, verticalAlignment=VAlign_Center}}
  }
  InputActionWidget {
   widgetClassPath=&CommonActionWidget
   bInherited=true
   bIsVariable=true
   parent=Prompt
   widget=InputActionWidget
   properties {
    progressMaterialParam=Percentage
    visibility=HitTestInvisible
    progressMaterialBrush {
     bIsDynamicallyLoaded=false
     drawAs=Image
     imageType=NoImage
     mirroring=NoMirror
     resourceName=None
     resourceObject=/FortUILibrary/UIKit/Buttons/MaterialInstances/CTAButtons/MI_UIKit_EmptyProgress
     tiling=NoTile
     imageSize=<48 48>
     margin=<ltrb 0 0 0 0>
     outlineSettings {
      bUseBrushTransparency=false
      roundingType=HalfHeightRadius
      width=0
      color {colorUseRule=UseColor_Specified, specifiedColor=<rgba 0 0 0 0>}
      cornerRadii=<0 0 0 0>
     }
     tintColor {colorUseRule=UseColor_Specified, specifiedColor=<rgba 1 1 1 1>}
     uVRegion {bIsValid=false, max=<0 0>, min=<0 0>}
    }
   }
   slot {class=GridSlot, properties {horizontalAlignment=HAlign_Center, layer=40, verticalAlignment=VAlign_Center}}
  }
 ]
}
