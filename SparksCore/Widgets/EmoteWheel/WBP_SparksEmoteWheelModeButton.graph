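# List-entry button for a mode option in the Sparks emote wheel, built on CommonButtonBase as a width-constrained, centered, uppercase text label.
# When its list item is set, it casts the item to SparksSongPartPlayerOverrideComponent and shows that component's override display name, or "Jam Loops" if the cast fails.
# Selection changes switch the text color to black when selected and white otherwise.
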WidgetBlueprint /SparksCore/Widgets/EmoteWheel/WBP_SparksEmoteWheelModeButton
graphs [EventGraph {nodes [
   n1 = Event On Item Selection Changed!() -> (OutputDelegate, then, bIsSelected:bool) {is=title K2Node_Event}
   > SetColorAndOpacity!(
    execute
    self=Get MyText() -> (MyText:Common Text Object Reference)
    InColorAndOpacity_SpecifiedColor:Linear Color Structure=Select(Option 0:Linear Color Structure=(R=1,G=1,B=1,A=1), Option 1:Linear Color Structure=(R=0,G=0,B=0,A=1), Index=@n1.bIsSelected) -> (ReturnValue:Linear Color Structure)
    InColorAndOpacity_ColorUseRule:ESlateColorStylingMode Enum
   ) -> (then)
   n2 = Event OnListItemObjectSet()
   > n3 = Cast To SparksSongPartPlayerOverrideComponent!(execute, Object=@n2.ListItemObject) -> (then, CastFailed, AsSparks Song Part Player Override Component:Sparks Song Part Player Override Component Object Reference) {is=title K2Node_DynamicCast}
   > Set Text(
    Text=Get Override Display Name!(self:Sparks Song Part Player Override Component Object Reference=@n3) -> (ReturnValue:Text) {is=title K2Node_CallFunction}
    self=Get MyText()
   )
   Set Text(execute=@n3.CastFailed, Text="LOCTEXT(\"Jam Loops\")", self=Get MyText())
  ]}]
variables []
widgets {
 info {inheritedWidgetCount=1, parentClass=&CommonButtonBase, rootWidgetClass=&SizeBox, widgetCount=5}
 widgets [
  SizeBox_0 {
   widgetClassPath=&SizeBox
   widget=SizeBox_0
   properties {bOverride_MaxDesiredWidth=true, bOverride_MinDesiredWidth=true, maxDesiredWidth=340, minDesiredWidth=240}
  }
  Overlay_0 {widgetClassPath=&Overlay, parent=SizeBox_0, widget=Overlay_0, properties {bOverride_Cursor=true}, slot {class=SizeBoxSlot}}
  Background {
   widgetClassPath=&Border
   bIsVariable=true
   parent=Overlay_0
   widget=Background
   properties {
    verticalAlignment=VAlign_Center
    background {
     bIsDynamicallyLoaded=false
     drawAs=NoDrawType
     imageType=NoImage
     mirroring=NoMirror
     resourceName=None
     resourceObject=null
     tiling=NoTile
     imageSize=<32 32>
     margin=<ltrb 0 0 0 0>
     outlineSettings {
      bUseBrushTransparency=false
      roundingType=FixedRadius
      width=0
      color {colorUseRule=UseColor_Specified, specifiedColor=<rgba 0 0 0 0>}
      cornerRadii=<17 17 17 17>
     }
     tintColor {colorUseRule=UseColor_Specified, specifiedColor=<rgba 1 1 1 1>}
     uVRegion {bIsValid=false, max=<0 0>, min=<0 0>}
    }
    padding=<ltrb 0 0 0 0>
   }
   slot {class=OverlaySlot, properties {horizontalAlignment=HAlign_Fill, verticalAlignment=VAlign_Fill}}
  }
  MyText {
   widgetClassPath=&CommonTextBlock
   bIsVariable=true
   parent=Background
   widget=MyText
   properties {
    justification=Center
    minDesiredWidth=256
    scrollStyle=&TextScrollStyle-Base_C
    text="{Option}"
    textTransformPolicy=ToUpper
    font {
     bForceMonospaced=false
     bMaterialIsStencil=false
     fontMaterial=null
     fontObject=&BurbankSmall
     letterSpacing=0
     monospacedWidth=1
     size=24
     skewAmount=0
     typefaceFontName=Black
     outlineSettings {
      bApplyOutlineToDropShadows=false
      bMiteredCorners=false
      bSeparateFillAlpha=false
      outlineMaterial=null
      outlineSize=0
      outlineColor=<rgba 0 0 0 1>
     }
    }
    shadowOffset=<0 0>
   }
   slot {class=BorderSlot, properties {verticalAlignment=VAlign_Center, padding=<ltrb 0 0 0 0>}}
  }
  InputActionWidget {widgetClassPath=&CommonActionWidget, bInherited=true, widget=null}
 ]
}
