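# Sparks Songs emote wheel overlay that lets the player switch between song part player override modes.
# On activation it lists the owning controller's sorted override components, preselecting the forced-open or active one.
# A keybind or invisible touch button cycles selection, moving an active gameplay tag to the chosen component.
# That component's overlay widget loads asynchronously into a full-screen container, and the selector hides when only one mode exists.
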WidgetBlueprint /SparksCore/Widgets/EmoteWheel/BP_SparksSongsEmoteWheelOverlay
graphs [
 ClearOverlayWidget {nodes [
   ClearOverlayWidget() {is=title K2Node_FunctionEntry}
   > Set OverrideOverlayWidget(OverrideOverlayWidget:User Widget Object Reference) -> (Output_Get:User Widget Object Reference)
   > ClearChildren(self=Get OverlayWidgetContainer() -> (OverlayWidgetContainer:Canvas Panel Object Reference))
  ]}
 CycleMode {nodes [
   CycleMode() {is=title K2Node_FunctionEntry}
   > n1 = Sequence!(execute) -> (then_0, then_1, then_2:Exec)
   > n2 = Set CurrentItem(CurrentItem:Object Reference=Get Selected Item!(self:List View Object Reference=Get ModeSelectionList() -> (ModeSelectionList:Common List View Object Reference)) -> (ReturnValue:Object Reference) {is=title K2Node_CallFunction}) -> (Output_Get:Object Reference)
   > n3 = Is Valid(InputObject=@n2)
   Set DesiredIndex(execute=@n3.Is Not Valid, DesiredIndex:int=0) -> (Output_Get:int=0)
   n4 = Get CurrentItem() -> (CurrentItem:Object Reference)
   n5 = Get ModeSelectionList()
   Set DesiredIndex(
    execute=@n3.Is Valid
    DesiredIndex=int + int(A=GetIndexForItem!(self:List View Object Reference=@n5, Item:Object Reference=@n4) -> (ReturnValue:int=0))
   ) -> (Output_Get=0)
   SetSelectedIndex!(execute=@n1.then_2, self:List View Object Reference=Get ModeSelectionList(), Index:int=Get DesiredIndex() -> (DesiredIndex:int=0)) -> (then)
   n6 = Get ModeSelectionList()
   Branch(
    execute=@n1.then_1
    Condition="integer >= integer"(A=Get DesiredIndex() -> (DesiredIndex=0), B=GetNumItems!(self:List View Object Reference=@n6) -> (ReturnValue:int=0))
   )
   > Set DesiredIndex(DesiredIndex=0) -> (Output_Get=0)
  ]}
 EventGraph {nodes [
   n1 = "On Clicked (TouchOverlayButton)"!() -> (OutputDelegate, then, Button:Common Button Base Object Reference) {is=title K2Node_ComponentBoundEvent}
   CycleMode(execute=[@n1.then, @n4.then]) {is=title K2Node_CallFunction}
   n2 = "On Item Selection Changed (ModeSelectionList)"!() -> (OutputDelegate, then, Item:Object Reference, bIsSelected:bool) {is=title K2Node_ComponentBoundEvent}
   > n3 = Cast To SparksSongPartPlayerOverrideComponent!(execute, Object=@n2.Item) -> (then, CastFailed, AsSparks Song Part Player Override Component:Sparks Song Part Player Override Component Object Reference) {is=title K2Node_DynamicCast}
   > SwitchToMode!(
    execute
    self
    Id:Gameplay Tag Structure=Get Override Id!(self:Sparks Song Part Player Override Component Object Reference=@n3) -> (ReturnValue:Gameplay Tag Structure) {is=title K2Node_CallFunction}
   ) -> (then)
   SwitchToMode(execute=@n3.CastFailed)
   n4 = "On Clicked (Keybind_Holdable)"!() -> (OutputDelegate, then, Button:Common Button Base Object Reference) {is=title K2Node_ComponentBoundEvent}
   n5 = OnInputMethodChanged_Event!() -> (OutputDelegate, then, bNewInputType:ECommonInputType Enum) {is=title K2Node_CustomEvent}
   > SetVisibility(
    self=Get TouchOverlayButton() -> (TouchOverlayButton:WBP UIKit Button Generic Object Reference)
    InVisibility=Select!(
     MouseAndKeyboard:ESlateVisibility Enum=Collapsed
     Gamepad:ESlateVisibility Enum=Collapsed
     Touch:ESlateVisibility Enum=SelfHitTestInvisible
     Count:ESlateVisibility Enum=Collapsed
     Index:ECommonInputType Enum=@n5.bNewInputType
    ) -> (ReturnValue)
   )
   > SetVisibility(
    self=Get KeybindVisibilityControl() -> (KeybindVisibilityControl:Grid Panel Object Reference)
    InVisibility=Select(MouseAndKeyboard=SelfHitTestInvisible, Gamepad=SelfHitTestInvisible, Touch=Collapsed, Count=SelfHitTestInvisible, Index=@n5.bNewInputType)
   )
   Event Construct()
   n6 = IsInputMethodActive!(
    self:Common Input Object Reference=Get CommonInputSubsystem!(PlayerController:Player Controller Object Reference=GetOwningPlayer()) -> (ReturnValue:Common Input Object Reference) {is=title K2Node_GetSubsystemFromPC}
    InputMethod:ECommonInputType Enum=Touch
   ) -> (ReturnValue:bool)
   > SetVisibility(
    self=Get TouchOverlayButton()
    InVisibility=Select!(Option 0:ESlateVisibility Enum=Collapsed, Option 1:ESlateVisibility Enum=SelfHitTestInvisible, Index:bool=@n6) -> (ReturnValue)
   )
   > SetVisibility(self=Get KeybindVisibilityControl(), InVisibility=Select(Option 0=SelfHitTestInvisible, Option 1=Collapsed, Index=@n6))
   > Assign On Input Method Changed!(
    execute:Exec
    self:Common Input Object Reference=Get CommonInputSubsystem(PlayerController=GetOwningPlayer())
    Delegate:"Delegate (by ref)"=@n5
   ) -> (then:Exec) {is=title K2Node_AssignDelegate}
   Event On Deactivated()
   > ClearOverlayWidget() {is=title K2Node_CallFunction}
   K2Node_Event_2 = Event UpdateOverlayWidget!() -> (OutputDelegate, then, OverlayWidgetClass:"User Widget Soft Class Reference (by ref)") {is=title K2Node_CustomEvent}
   > ClearOverlayWidget()
   > Branch(Condition=IsValidSoftClassReference!(SoftClassReference:"Object Soft Class Reference (by ref)"=@K2Node_Event_2.OverlayWidgetClass) -> (ReturnValue:bool))
   > n7 = CreateWidgetAsync!(
    execute:Exec
    UserWidgetSoftClass:User Widget Soft Class Reference=@K2Node_Event_2.OverlayWidgetClass
    OwningPlayer:Player Controller Object Reference=GetOwningPlayer()
    bSuspendInputUntilComplete:bool=true
   ) -> (then:Exec, OnComplete:Exec, OnFailure:Exec, UserWidget:User Widget Object Reference) {is=title K2Node_AsyncAction}
   n8 = Set OverrideOverlayWidget(execute=@n7.OnComplete, OverrideOverlayWidget=@n7)
   > n9 = AddChildToCanvas!(execute, self:Canvas Panel Object Reference=Get OverlayWidgetContainer(), Content:Widget Object Reference=@n8) -> (then, ReturnValue:Canvas Panel Slot Object Reference)
   > SetAnchors!(
    execute
    self:Canvas Panel Slot Object Reference=@n9
    InAnchors_Minimum:Vector 2D Structure=(X=0,Y=0)
    InAnchors_Maximum:Vector 2D Structure=(X=1,Y=1)
   ) -> (then)
   > SetOffsets!(
    execute
    self:Canvas Panel Slot Object Reference=@n9
    InOffset_Left:float=0
    InOffset_Top:float=0
    InOffset_Right:float=0
    InOffset_Bottom:float=0
   ) -> (then)
   Event On Activated()
   > PopulateOverrides()
  ]}
 GetActiveModeIndex {nodes [
   GetActiveModeIndex() {is=title K2Node_FunctionEntry}
   > n1 = For Each Loop(Array:Array of Object References=GetListItems!(self:List View Object Reference=Get ModeSelectionList()) -> (ReturnValue:Array of Object References)) -> (Array Element:Object Reference)
   Return Node!(execute=@n1.Completed, ModeIndex:int=0)
   > n2 = Cast To SparksSongPartPlayerOverrideComponent(Object=@n1)
   > Branch(Condition=IsOverrideActive!(self:Sparks Song Part Player Override Component Object Reference=@n2) -> (ReturnValue:bool))
   > Return Node(ModeIndex=@n1.Array Index)
  ]}
 PopulateOverrides {nodes [
   PopulateOverrides() {is=title K2Node_FunctionEntry}
   > n1 = SortOverrideComponents!(
    execute
    self:Sparks Songs Emote Wheel Overlay Object Reference
    Components:Array of Sparks Song Part Player Override Component Object References=Get Components By Class!(
     self:Actor Object Reference=GetOwningPlayer()
     ComponentClass:Actor Component Class Reference=/Script/SparksCoreCosmeticsRuntime.SparksSongPartPlayerOverrideComponent
    ) -> (ReturnValue:Array of Sparks Song Part Player Override Component Object References)
   ) -> (then, ReturnValue:Array of Sparks Song Part Player Override Component Object References)
   > n2 = For Each Loop(Array:Array of Sparks Song Part Player Override Component Object References=@n1) -> (Array Element:Sparks Song Part Player Override Component Object Reference)
   n3 = Get ForcedOpenIndex() -> (ForcedOpenIndex:int=0)
   n4 = Branch(execute=@n2.Completed, Condition="integer >= integer"(A=@n3))
   > SetSelectedIndex(self=Get ModeSelectionList(), Index=@n3)
   n5 = GetActiveModeIndex!(execute=@n4.else, self) -> (then, ModeIndex:int=0) {is=title K2Node_CallFunction}
   > SetSelectedIndex(self=Get ModeSelectionList(), Index=@n5)
   AddItem!(execute=@n2.LoopBody, self:List View Object Reference=Get ModeSelectionList(), Item:Object Reference=@n2) -> (then)
   > Branch(Condition="integer < integer"(A=Get ForcedOpenIndex() -> (ForcedOpenIndex=0)))
   > Branch(Condition=IsForcedOpen!(self:Sparks Song Part Player Override Component Object Reference=@n2) -> (ReturnValue:bool))
   > Set ForcedOpenIndex(ForcedOpenIndex:int=@n2.Array Index) -> (Output_Get:int=0)
  ]}
 SwitchToMode {nodes [
   n1 = SwitchToMode!() -> (then, Id:Gameplay Tag Structure) {is=title K2Node_FunctionEntry}
   > Set CurrentOverride(CurrentOverride:Sparks Song Part Player Override Component Object Reference) -> (Output_Get:Sparks Song Part Player Override Component Object Reference)
   > n2 = For Each Loop(Array:Array of Object References=GetListItems(self=Get ModeSelectionList())) -> (Array Element:Object Reference)
   UpdateOverlayVisibility(execute=@n2.Completed)
   n3 = Cast To SparksSongPartPlayerOverrideComponent(execute=@n2.LoopBody, Object=@n2)
   > n4 = Branch(Condition="Equal (GameplayTag)"(A=Get Override Id(self=@n3), B=@n1))
   n5 = Get OverrideActiveTag() -> (OverrideActiveTag:Gameplay Tag Structure)
   > AddGameplayTag!(execute, self:Sparks Song Part Player Override Component Object Reference=@n3, Tag:"Gameplay Tag Structure (by ref)"=@n5) -> (then, ReturnValue:bool)
   > Set CurrentOverride(CurrentOverride=@n3)
   RemoveGameplayTag!(execute=@n4.else, self:Sparks Song Part Player Override Component Object Reference=@n3, Tag:"Gameplay Tag Structure (by ref)"=@n5) -> (then, ReturnValue:bool)
  ]}
 UpdateOverlayVisibility {nodes [
   UpdateOverlayVisibility() {is=title K2Node_FunctionEntry}
   > n1 = Sequence!(execute) -> (then_0, then_1)
   n2 = Get ModeSelectionList()
   > n3 = Branch(Condition="integer <= integer"(A=GetNumItems(self=@n2) -> (ReturnValue=0), B=1))
   n4 = Get GridPanel_ModeSelection() -> (GridPanel_ModeSelection:Grid Panel Object Reference)
   > SetVisibility(self=@n4, InVisibility=Hidden)
   SetVisibility(execute=@n3.else, self=@n4, InVisibility=SelfHitTestInvisible)
   n5 = Get CurrentOverride() -> (CurrentOverride:Sparks Song Part Player Override Component Object Reference)
   n6 = Is Valid(exec=@n1.then_1, InputObject=@n5)
   > n7 = Branch(Condition=IsOverrideActive(self=@n5))
   n8 = Get Class(Object=Get OverrideOverlayWidget() -> (OverrideOverlayWidget:User Widget Object Reference)) -> (ReturnValue:User Widget Class Reference)
   n9 = Get Legacy Overlay Widget!(self:Sparks Song Part Player Override Component Object Reference=@n5) -> (ReturnValue:Common Activatable Widget Soft Class Reference) {is=title K2Node_CallFunction}
   > n10 = Branch(Condition="Equal (SoftClassReference)"(
     A:"Object Soft Class Reference (by ref)"=@n9
     B:"Object Soft Class Reference (by ref)"=Make Soft Reference!(Input:User Widget Class Reference=@n8) -> (Output:User Widget Soft Class Reference)
    ) -> (ReturnValue:bool=~) {is=title K2Node_PromotableOperator})
   Event UpdateOverlayWidget!(execute=@n10.else, self, OverlayWidgetClass:"User Widget Soft Class Reference (by ref)"=@n9) -> (then) {is=title K2Node_CallFunction}
   ClearOverlayWidget(execute=[@n7.else, @n6.Is Not Valid])
  ]}
]
variables [
 CurrentOverride {type=SparksSongPartPlayerOverrideComponent, default=null}
 OverrideActiveTag {type=GameplayTag, default {tagName=Sparks.SongPartPlayerOverride.Active.SparksSongsEmoteWheel}}
 OverrideOverlayWidget {type=UserWidget, default=null}
]
widgets {
 info {
  inheritedWidgetCount=0
  parentClass=/Script/SparksSongsEmoteWheelUI.SparksSongsEmoteWheelOverlay
  rootWidgetClass=&CanvasPanel
  widgetCount=9
 }
 widgets [
  CanvasPanel_0 {widgetClassPath=&CanvasPanel, widget=CanvasPanel_0}
  GridPanel_ModeSelection {
   widgetClassPath=&GridPanel
   bIsVariable=true
   parent=CanvasPanel_0
   widget=GridPanel_ModeSelection
   properties {columnFill [1], rowFill [1]}
   slot {
    class=CanvasPanelSlot
    properties {bAutoSize=true, layoutData {alignment=<0 0.5>, anchors {maximum=<0 0.5>, minimum=<0 0.5>}, offsets=<ltrb 100 0 650.585 270.6306>}}
   }
  }
  Background {
   widgetClassPath=&Image
   bIsVariable=true
   parent=GridPanel_ModeSelection
   widget=Background
   properties {
    renderOpacity=0.5
    brush {
     bIsDynamicallyLoaded=false
     drawAs=Image
     imageType=FullColor
     mirroring=NoMirror
     resourceName=None
     resourceObject=/SparksCore/Widgets/Materials/MI_UIKit_BackplateCombined
     tiling=NoTile
     imageSize=<32 32>
     margin=<ltrb 0 0 0 0>
     outlineSettings {
      bUseBrushTransparency=false
      roundingType=HalfHeightRadius
      width=0
      color {colorUseRule=UseColor_Specified, specifiedColor=<rgba 0 0 0 0>}
      cornerRadii=<0 0 0 0>
     }
     tintColor {colorUseRule=UseColor_Specified, specifiedColor=<rgba 1 1 1 1>}
     uVRegion {bIsValid=false, max=<0 0>, min=<0 0>}
    }
   }
   slot {class=GridSlot}
  }
  SizeBox_0 {
   widgetClassPath=&SizeBox
   parent=GridPanel_ModeSelection
   widget=SizeBox_0
   properties {bOverride_MinDesiredHeight=true, bOverride_MinDesiredWidth=true, minDesiredHeight=96, minDesiredWidth=256}
   slot {class=GridSlot, properties {verticalAlignment=VAlign_Top, padding=<ltrb 8 8 8 8>}}
  }
  ModeSelectionList {
   widgetClassPath=&CommonListView
   bIsVariable=true
   parent=SizeBox_0
   widget=ModeSelectionList
   properties {
    allowOverscroll=false
    bAllowDragging=false
    bEnableRightClickScrolling=false
    bEnableScrollAnimation=false
    bEnableTouchScrolling=false
    entryWidgetClass=&/WBP_SparksEmoteWheelModeButton.WBP_SparksEmoteWheelModeButton_C
    horizontalEntrySpacing=10
    numDesignerPreviewEntries=2
   }
   slot {class=SizeBoxSlot}
  }
  KeybindVisibilityControl {
   widgetClassPath=&GridPanel
   bIsVariable=true
   parent=GridPanel_ModeSelection
   widget=KeybindVisibilityControl
   properties {columnFill [1], rowFill [1]}
   slot {class=GridSlot, properties {horizontalAlignment=HAlign_Right, verticalAlignment=VAlign_Center, padding=<ltrb 0 0 -60 0>}}
  }
  Keybind_Holdable {
   widgetClassPath=/FortUILibrary/Systems/ModularButton/WBP_UIKit_Keybind_Holdable
   bIsVariable=true
   parent=KeybindVisibilityControl
   widget=Keybind_Holdable
   properties {bRequiresHold=false, triggeringInputAction {dataTable=/SparksCore/Input/SparksSongsEmoteWheelBindings, rowName=SwapSongOverrideMode}}
   slot {class=GridSlot}
  }
  TouchOverlayButton {
   widgetClassPath=&WBP_UIKit_Button_Generic
   bIsVariable=true
   parent=GridPanel_ModeSelection
   widget=TouchOverlayButton
   properties {
    renderOpacity=0
    style=/FortUILibrary/Foundation/Buttons/ButtonStyles/ButtonStyle-Empty.ButtonStyle-Empty_C
    useIconBlock=false
    useTextBlock=false
    visibility=Collapsed
   }
   slot {class=GridSlot, properties {layer=10}}
  }
  OverlayWidgetContainer {
   widgetClassPath=&CanvasPanel
   bIsVariable=true
   parent=CanvasPanel_0
   widget=OverlayWidgetContainer
   slot {class=CanvasPanelSlot, properties {layoutData {alignment=<0 0>, anchors {maximum=<1 1>, minimum=<0 0>}, offsets=<ltrb 0 0 0 0>}}}
  }
 ]
}
