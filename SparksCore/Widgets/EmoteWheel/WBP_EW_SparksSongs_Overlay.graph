WidgetBlueprint /SparksCore/Widgets/EmoteWheel/WBP_EW_SparksSongs_Overlay
animations [Intro {
  binding_count=1
  length_seconds=0.15
  object=&self:Intro
  sequence {bindings [Overlay_0 {
     id=#1
     possessed_class=Overlay
     tracks [{
       type=MovieSceneFloatTrack
       display_name=Render Opacity
       sections [{type=MovieSceneFloatSection, end_frame=3, start_frame=0, channels [None {type=MovieSceneScriptingFloatChannel, keys [[0,0],[3,1]]}]}]
      }]
    }]}
 }]
graphs [
 ClearOverlayWidget {nodes [
   ClearOverlayWidget() {is=title K2Node_FunctionEntry}
   > n1 = Get!(execute:Exec) -> (then:Exec, else:Exec, OverrideOverlayWidget:User Widget Object Reference)
   > RemoveChild(self=Get OverlayWidgetContainer() -> (OverlayWidgetContainer:Overlay Object Reference), Content=@n1)
   > Set OverrideOverlayWidget(OverrideOverlayWidget:User Widget Object Reference) -> (Output_Get:User Widget Object Reference)
  ]}
 CycleMode {nodes [
   CycleMode() {is=title K2Node_FunctionEntry}
   > n1 = Sequence!(execute) -> (then_0, then_1, then_2:Exec)
   > n2 = Set CurrentItem(CurrentItem:Object Reference=Get Selected Item!(self:List View Object Reference=Get ModeSelectionList() -> (ModeSelectionList:Common List View Object Reference)) -> (ReturnValue:Object Reference) {is=title K2Node_CallFunction}) -> (Output_Get:Object Reference)
   > n3 = Is Valid(InputObject=@n2)
   Set DesiredIndex(execute=@n3.Is Not Valid, DesiredIndex:int=0) -> (Output_Get:int=0)
   n4 = Get CurrentItem() -> (CurrentItem:Object Reference)
   n5 = Get ModeSelectionList()
   Set DesiredIndex(
    execute=@n3.Is Valid
    DesiredIndex=int + int(A=GetIndexForItem!(self:List View Object Reference=@n5, Item:Object Reference=@n4) -> (ReturnValue:int=0))
   ) -> (Output_Get=0)
   SetSelectedIndex!(execute=@n1.then_2, self:List View Object Reference=Get ModeSelectionList(), Index:int=Get DesiredIndex() -> (DesiredIndex:int=0)) -> (then)
   n6 = Get ModeSelectionList()
   Branch(
    execute=@n1.then_1
    Condition="integer >= integer"(A=Get DesiredIndex() -> (DesiredIndex=0), B=GetNumItems!(self:List View Object Reference=@n6) -> (ReturnValue:int=0))
   )
   > Set DesiredIndex(DesiredIndex=0) -> (Output_Get=0)
  ]}
 EventGraph {nodes [
   n1 = "On Clicked (TouchOverlayButton)"!() -> (OutputDelegate, then, Button:Common Button Base Object Reference) {is=title K2Node_ComponentBoundEvent}
   CycleMode(execute=[@n1.then, @n4.then]) {is=title K2Node_CallFunction}
   n2 = "On Item Selection Changed (ModeSelectionList)"!() -> (OutputDelegate, then, Item:Object Reference, bIsSelected:bool) {is=title K2Node_ComponentBoundEvent}
   > n3 = Cast To SparksSongPartPlayerOverrideComponent!(execute, Object=@n2.Item) -> (then, CastFailed, AsSparks Song Part Player Override Component:Sparks Song Part Player Override Component Object Reference) {is=title K2Node_DynamicCast}
   > SwitchToMode!(
    execute
    self
    Id:Gameplay Tag Structure=Get Override Id!(self:Sparks Song Part Player Override Component Object Reference=@n3) -> (ReturnValue:Gameplay Tag Structure) {is=title K2Node_CallFunction}
   ) -> (then)
   SwitchToMode(execute=@n3.CastFailed)
   n4 = "On Clicked (Keybind_Holdable)"!() -> (OutputDelegate, then, Button:Common Button Base Object Reference) {is=title K2Node_ComponentBoundEvent}
   n5 = OnInputMethodChanged_Event!() -> (OutputDelegate, then, bNewInputType:ECommonInputType Enum) {is=title K2Node_CustomEvent}
   > SetVisibility(
    self=Get TouchOverlayButton() -> (TouchOverlayButton:WBP UIKit Button Generic Object Reference)
    InVisibility=Select!(
     MouseAndKeyboard:ESlateVisibility Enum=Collapsed
     Gamepad:ESlateVisibility Enum=Collapsed
     Touch:ESlateVisibility Enum=SelfHitTestInvisible
     Count:ESlateVisibility Enum=Collapsed
     Index:ECommonInputType Enum=@n5.bNewInputType
    ) -> (ReturnValue)
   )
   > SetVisibility(
    self=Get KeybindVisibilityControl() -> (KeybindVisibilityControl:Grid Panel Object Reference)
    InVisibility=Select(MouseAndKeyboard=SelfHitTestInvisible, Gamepad=SelfHitTestInvisible, Touch=Collapsed, Count=SelfHitTestInvisible, Index=@n5.bNewInputType)
   )
   n6 = ModeSelectionListItemChanged_Event!() -> (OutputDelegate, then, Item:Object Reference, bIsSelected:bool=~) {is=title K2Node_CustomEvent}
   > Branch(Condition=@n6.bIsSelected)
   n7 = "integer < integer"(A=GetIndexForItem(self=Get ModeSelectionList(), Item=@n6.Item) -> (ReturnValue=0), B=1)
   > SetRenderScale!(
    execute
    self:Widget Object Reference=Get ModeSelection_Item() -> (ModeSelection_Item:Image Object Reference)
    Scale:Vector 2D Structure=MakeVector2D(X=SelectFloat(A=-1, B=1, bPickA=@n7), Y=1)
   ) -> (then)
   > SetScalarParameterValue(
    self=Get JTP_DynamicMat() -> (JTP_DynamicMat:Material Instance Dynamic Object Reference)
    ParameterName=b_Selected
    Value="To Float (Boolean)"(InBool=NOT Boolean(A=@n7))
   )
   > SetScalarParameterValue(
    self=Get Jam_DynamicMat() -> (Jam_DynamicMat:Material Instance Dynamic Object Reference)
    ParameterName=b_Selected
    Value="To Float (Boolean)"(InBool=@n7)
   )
   Event Construct()
   n8 = IsInputMethodActive!(
    self:Common Input Object Reference=Get CommonInputSubsystem!(PlayerController:Player Controller Object Reference=GetOwningPlayer()) -> (ReturnValue:Common Input Object Reference) {is=title K2Node_GetSubsystemFromPC}
    InputMethod:ECommonInputType Enum=Touch
   ) -> (ReturnValue:bool)
   > SetVisibility(
    self=Get TouchOverlayButton()
    InVisibility=Select!(Option 0:ESlateVisibility Enum=Collapsed, Option 1:ESlateVisibility Enum=SelfHitTestInvisible, Index:bool=@n8) -> (ReturnValue)
   )
   > SetVisibility(self=Get KeybindVisibilityControl(), InVisibility=Select(Option 0=SelfHitTestInvisible, Option 1=Collapsed, Index=@n8))
   > Assign On Input Method Changed!(
    execute:Exec
    self:Common Input Object Reference=Get CommonInputSubsystem(PlayerController=GetOwningPlayer())
    Delegate:"Delegate (by ref)"=@n5
   ) -> (then:Exec) {is=title K2Node_AssignDelegate}
   > n9 = Sequence()
   > Bind Event to On Item Selection Changed(self:List View Object Reference=Get ModeSelectionList(), Delegate=Create Event()) {is=title K2Node_AddDelegate}
   n10 = GetDynamicMaterial(execute=@n9.then_1, self=Get Image_JTP() -> (Image_JTP:Image Object Reference))
   > Set JTP_DynamicMat(JTP_DynamicMat:Material Instance Dynamic Object Reference=@n10) -> (Output_Get:Material Instance Dynamic Object Reference)
   n11 = GetDynamicMaterial(execute=@n9.then_2, self=Get Image_Jam() -> (Image_Jam:Image Object Reference))
   > Set Jam_DynamicMat(Jam_DynamicMat:Material Instance Dynamic Object Reference=@n11) -> (Output_Get:Material Instance Dynamic Object Reference)
   Event On Deactivated()
   > ClearOverlayWidget() {is=title K2Node_CallFunction}
   K2Node_Event_2 = Event UpdateOverlayWidget!() -> (OutputDelegate, then, OverlayWidgetClass:"User Widget Soft Class Reference (by ref)") {is=title K2Node_CustomEvent}
   > ClearOverlayWidget()
   > Branch(Condition=IsValidSoftClassReference!(SoftClassReference:"Object Soft Class Reference (by ref)"=@K2Node_Event_2.OverlayWidgetClass) -> (ReturnValue:bool))
   > n12 = CreateWidgetAsync!(
    execute:Exec
    UserWidgetSoftClass:User Widget Soft Class Reference=@K2Node_Event_2.OverlayWidgetClass
    OwningPlayer:Player Controller Object Reference=GetOwningPlayer()
    bSuspendInputUntilComplete:bool=true
   ) -> (then:Exec, OnComplete:Exec, OnFailure:Exec, UserWidget:User Widget Object Reference) {is=title K2Node_AsyncAction}
   n13 = Set OverrideOverlayWidget(execute=@n12.OnComplete, OverrideOverlayWidget=@n12)
   > AddChildToOverlay!(execute, self:Overlay Object Reference=Get OverlayWidgetContainer(), Content:Widget Object Reference=@n13) -> (then, ReturnValue:Overlay Slot Object Reference)
   n14 = SlotAsOverlaySlot(Widget=@n13)
   > SetHorizontalAlignment(self=@n14, InHorizontalAlignment=HAlign_Fill)
   > SetVerticalAlignment(self=@n14, InVerticalAlignment=VAlign_Fill)
   Event Destruct()
   > Unbind Event from On Item Selection Changed(self:List View Object Reference=Get ModeSelectionList(), Delegate=Create Event()) {is=title K2Node_RemoveDelegate}
   Event On Activated()
   > PopulateOverrides()
  ]}
 GetActiveModeIndex {nodes [
   GetActiveModeIndex() {is=title K2Node_FunctionEntry}
   > n1 = For Each Loop(Array:Array of Object References=GetListItems!(self:List View Object Reference=Get ModeSelectionList()) -> (ReturnValue:Array of Object References)) -> (Array Element:Object Reference)
   Return Node!(execute=@n1.Completed, ModeIndex:int=0)
   > n2 = Cast To SparksSongPartPlayerOverrideComponent(Object=@n1)
   > Branch(Condition=IsOverrideActive!(self:Sparks Song Part Player Override Component Object Reference=@n2) -> (ReturnValue:bool))
   > Return Node(ModeIndex=@n1.Array Index)
  ]}
 PopulateOverrides {nodes [
   PopulateOverrides() {is=title K2Node_FunctionEntry}
   > n1 = SortOverrideComponents!(
    execute
    self:Sparks Songs Emote Wheel Overlay Object Reference
    Components:Array of Sparks Song Part Player Override Component Object References=Get Components By Class!(
     self:Actor Object Reference=GetOwningPlayer()
     ComponentClass:Actor Component Class Reference=/Script/SparksCoreCosmeticsRuntime.SparksSongPartPlayerOverrideComponent
    ) -> (ReturnValue:Array of Sparks Song Part Player Override Component Object References)
   ) -> (then, ReturnValue:Array of Sparks Song Part Player Override Component Object References)
   > n2 = For Each Loop(Array:Array of Sparks Song Part Player Override Component Object References=@n1) -> (Array Element:Sparks Song Part Player Override Component Object Reference)
   n3 = Get ForcedOpenIndex() -> (ForcedOpenIndex:int=0)
   n4 = Branch(execute=@n2.Completed, Condition="integer >= integer"(A=@n3))
   > SetSelectedIndex(self=Get ModeSelectionList(), Index=@n3)
   n5 = GetActiveModeIndex!(execute=@n4.else, self) -> (then, ModeIndex:int=0) {is=title K2Node_CallFunction}
   > SetSelectedIndex(self=Get ModeSelectionList(), Index=@n5)
   AddItem!(execute=@n2.LoopBody, self:List View Object Reference=Get ModeSelectionList(), Item:Object Reference=@n2) -> (then)
   > Branch(Condition="integer < integer"(A=Get ForcedOpenIndex() -> (ForcedOpenIndex=0)))
   > Branch(Condition=IsForcedOpen!(self:Sparks Song Part Player Override Component Object Reference=@n2) -> (ReturnValue:bool))
   > Set ForcedOpenIndex(ForcedOpenIndex:int=@n2.Array Index) -> (Output_Get:int=0)
  ]}
 SetLayout {nodes [
   n1 = SetLayout!() -> (then, ShowContainer:bool=~) {is=title K2Node_FunctionEntry}
   > n2 = Set IsShowingCollection(IsShowingCollection:bool=@n1) -> (Output_Get:bool)
   n3 = SlotAsOverlaySlot(Widget=Get GridPanel_ModeSelection() -> (GridPanel_ModeSelection:Grid Panel Object Reference))
   > SetHorizontalAlignment(
    self=@n3
    InHorizontalAlignment=Select(Option 0:EHorizontalAlignment Enum=HAlign_Fill, Option 1:EHorizontalAlignment Enum=HAlign_Left, Index=@n2) -> (ReturnValue:EHorizontalAlignment Enum)
   )
   > SetVerticalAlignment(
    self=@n3
    InVerticalAlignment=Select(Option 0:EVerticalAlignment Enum=VAlign_Fill, Option 1:EVerticalAlignment Enum=VAlign_Top, Index=@n2) -> (ReturnValue:EVerticalAlignment Enum)
   )
   > SetPadding!(
    execute
    self:Overlay Slot Object Reference=@n3
    InPadding:Margin Structure=Make Margin(
     Left=Select(Option 0:float=0, Option 1:float=84, Index=@n2) -> (ReturnValue:float)
     Top=Select(Option 0=0, Option 1=114, Index=@n2)
     Right=0
     Bottom=0
    )
   ) -> (then)
   > SetVisibility(
    self=Get Border_CollectionView() -> (Border_CollectionView:Border Object Reference)
    InVisibility=Select(Option 0:ESlateVisibility Enum=Hidden, Option 1:ESlateVisibility Enum=SelfHitTestInvisible, Index=@n2) -> (ReturnValue:ESlateVisibility Enum)
   )
   > n4 = Get()
   n5 = GetNumItems(self=Get ModeSelectionList()) -> (ReturnValue=0)
   > Set isSingleCatCollectionScreen(isSingleCatCollectionScreen:bool=AND Boolean(A=Get IsShowingCollection() -> (IsShowingCollection:bool), B="integer <= integer"(A=@n5, B=1))) -> (Output_Get:bool)
   > SetPadding(
    self=SlotAsOverlaySlot(Widget=@n4)
    InPadding=Select(
     Option 0:Margin Structure=Get WheelOverlayPadding() -> (WheelOverlayPadding:Margin Structure)
     Option 1:Margin Structure=Get CollectionOverlayPadding() -> (CollectionOverlayPadding:Margin Structure)
     Index=Get isSingleCatCollectionScreen() -> (isSingleCatCollectionScreen:bool)
    ) -> (ReturnValue:Margin Structure)
   )
   > SetPadding(
    self=SlotAsOverlaySlot(Widget=Get Border_CollectionView())
    InPadding=Select(
     Option 0=Get BackgroundOverlayPaddingMultiCat() -> (BackgroundOverlayPaddingMultiCat:Margin Structure)
     Option 1=Get BackgroundOverlayPaddingSingleCat() -> (BackgroundOverlayPaddingSingleCat:Margin Structure)
     Index=Get isSingleCatCollectionScreen()
    )
   )
  ]}
 SwitchToMode {nodes [
   n1 = SwitchToMode!() -> (then, Id:Gameplay Tag Structure) {is=title K2Node_FunctionEntry}
   > Set CurrentOverride(CurrentOverride:Sparks Song Part Player Override Component Object Reference) -> (Output_Get:Sparks Song Part Player Override Component Object Reference)
   > n2 = For Each Loop(Array:Array of Object References=GetListItems(self=Get ModeSelectionList())) -> (Array Element:Object Reference)
   UpdateOverlayVisibility(execute=@n2.Completed)
   n3 = Cast To SparksSongPartPlayerOverrideComponent(execute=@n2.LoopBody, Object=@n2)
   > n4 = Branch(Condition="Equal (GameplayTag)"(A=Get Override Id(self=@n3), B=@n1))
   n5 = Get OverrideActiveTag() -> (OverrideActiveTag:Gameplay Tag Structure)
   > AddGameplayTag!(execute, self:Sparks Song Part Player Override Component Object Reference=@n3, Tag:"Gameplay Tag Structure (by ref)"=@n5) -> (then, ReturnValue:bool)
   > Set CurrentOverride(CurrentOverride=@n3)
   > Set Text(
    Text=Get SongPartPlayerText() -> (SongPartPlayerText:Text)
    self=Get ChooseElement_Text() -> (ChooseElement_Text:Common Text Object Reference)
   )
   RemoveGameplayTag!(execute=@n4.else, self:Sparks Song Part Player Override Component Object Reference=@n3, Tag:"Gameplay Tag Structure (by ref)"=@n5) -> (then, ReturnValue:bool)
   > Set Text(Text=Get SongPlayerText() -> (SongPlayerText:Text), self=Get ChooseElement_Text())
  ]}
 UpdateOverlayTextVisibility {nodes [UpdateOverlayTextVisibility() {is=title K2Node_FunctionEntry}]}
 UpdateOverlayVisibility {nodes [
   UpdateOverlayVisibility() {is=title K2Node_FunctionEntry}
   > n1 = Sequence!(execute) -> (then_0, then_1)
   n2 = Get ModeSelectionList()
   > n3 = Branch(Condition="integer <= integer"(A=GetNumItems(self=@n2) -> (ReturnValue=0), B=1))
   n4 = Get OverlayModeToggle() -> (OverlayModeToggle:Overlay Object Reference)
   > SetVisibility(self=@n4, InVisibility=Collapsed)
   SetVisibility(execute=@n3.else, self=@n4, InVisibility=SelfHitTestInvisible)
   n5 = Get CurrentOverride() -> (CurrentOverride:Sparks Song Part Player Override Component Object Reference)
   n6 = Is Valid(exec=@n1.then_1, InputObject=@n5)
   > n7 = Branch(Condition=IsOverrideActive(self=@n5))
   n8 = Get Class(Object=Get OverrideOverlayWidget() -> (OverrideOverlayWidget:User Widget Object Reference)) -> (ReturnValue:User Widget Class Reference)
   n9 = Get Overlay Widget!(self:Sparks Song Part Player Override Component Object Reference=@n5) -> (ReturnValue:Common Activatable Widget Soft Class Reference) {is=title K2Node_CallFunction}
   > n10 = Branch(Condition="Equal (SoftClassReference)"(
     A:"Object Soft Class Reference (by ref)"=@n9
     B:"Object Soft Class Reference (by ref)"=Make Soft Reference!(Input:User Widget Class Reference=@n8) -> (Output:User Widget Soft Class Reference)
    ) -> (ReturnValue:bool=~) {is=title K2Node_PromotableOperator})
   Event UpdateOverlayWidget!(execute=@n10.else, self, OverlayWidgetClass:"User Widget Soft Class Reference (by ref)"=@n9) -> (then) {is=title K2Node_CallFunction}
   ClearOverlayWidget(execute=[@n7.else, @n6.Is Not Valid])
  ]}
]
variables [
 BackgroundOverlayPaddingMultiCat {type=Margin, default=<ltrb -20 -40 -20 44>}
 BackgroundOverlayPaddingSingleCat {type=Margin, default=<ltrb -20 -40 -20 140>}
 CollectionOverlayPadding {type=Margin, default=<ltrb 0 -96 0 0>}
 CurrentOverride {type=SparksSongPartPlayerOverrideComponent, default=null}
 IsShowingCollection {type=boolean}
 Jam_DynamicMat {type=MaterialInstanceDynamic}
 JTP_DynamicMat {type=MaterialInstanceDynamic}
 OverrideActiveTag {type=GameplayTag, default {tagName=Sparks.SongPartPlayerOverride.Active.SparksSongsEmoteWheel}}
 OverrideOverlayWidget {type=UserWidget, default=null}
 SongPartPlayerText {type=string, default=Choose Your Instrument}
 SongPlayerText {type=string, default=Choose Your Song}
 WheelOverlayPadding {type=Margin, default=<ltrb 0 0 0 0>}
]
widgets {
 info {
  inheritedWidgetCount=0
  parentClass=/Script/SparksSongsEmoteWheelUI.SparksSongsEmoteWheelOverlay
  rootWidgetClass=&Overlay
  widgetCount=17
 }
 widgets [
  Overlay_0 {widgetClassPath=&Overlay, widget=Overlay_0, properties {bOverride_Cursor=true}}
  OverlayWidgetContainer {
   widgetClassPath=&Overlay
   bIsVariable=true
   parent=Overlay_0
   widget=OverlayWidgetContainer
   properties {bOverride_Cursor=true}
   slot {class=OverlaySlot, properties {horizontalAlignment=HAlign_Fill, verticalAlignment=VAlign_Fill}}
  }
  GridPanel_ModeSelection {
   widgetClassPath=&GridPanel
   bIsVariable=true
   parent=OverlayWidgetContainer
   widget=GridPanel_ModeSelection
   properties {columnFill [1], rowFill [1]}
   slot {class=OverlaySlot, properties {horizontalAlignment=HAlign_Fill, verticalAlignment=VAlign_Fill}}
  }
  SizeBox_0 {
   widgetClassPath=&SizeBox
   parent=GridPanel_ModeSelection
   widget=SizeBox_0
   properties {
    bOverride_HeightOverride=true
    bOverride_WidthOverride=true
    heightOverride=300
    maxDesiredWidth=340
    minDesiredHeight=64
    minDesiredWidth=352
    widthOverride=300
   }
   slot {class=GridSlot, properties {horizontalAlignment=HAlign_Center, verticalAlignment=VAlign_Center}}
  }
  OverlaySelectionUI {widgetClassPath=&Overlay, parent=SizeBox_0, widget=OverlaySelectionUI, slot {class=SizeBoxSlot}}
  Border_CollectionView {
   widgetClassPath=&Border
   bIsVariable=true
   parent=OverlaySelectionUI
   widget=Border_CollectionView
   properties {
    verticalAlignment=VAlign_Top
    visibility=SelfHitTestInvisible
    background {
     bIsDynamicallyLoaded=false
     drawAs=Box
     imageType=FullColor
     mirroring=NoMirror
     resourceName=None
     resourceObject=/FortUILibrary/Systems/Backplate/MaterialInstances/MI_UIKit_Backplate
     tiling=NoTile
     imageSize=<32 32>
     margin=<ltrb 0 0 0 0>
     outlineSettings {
      bUseBrushTransparency=false
      roundingType=HalfHeightRadius
      width=0
      color {colorUseRule=UseColor_Specified, specifiedColor=<rgba 0 0 0 0>}
      cornerRadii=<0 0 0 0>
     }
     tintColor {colorUseRule=UseColor_Specified, specifiedColor=<rgba 1 1 1 1>}
     uVRegion {bIsValid=false, max=<0 0>, min=<0 0>}
    }
    brushColor=<rgba 0 0 0 0.4>
    padding=<ltrb 20 10 20 0>
   }
   slot {class=OverlaySlot, properties {horizontalAlignment=HAlign_Fill, verticalAlignment=VAlign_Fill, padding=<ltrb -20 -40 -20 44>}}
  }
  ChooseElement_Text {
   widgetClassPath=&CommonTextBlock
   bIsVariable=true
   parent=Border_CollectionView
   widget=ChooseElement_Text
   properties {
    autoWrapText=true
    bIsScrollingEnabled=false
    justification=Center
    scrollStyle=&TextScrollStyle-Base_C
    text=Choose Your Instrument
    colorAndOpacity {colorUseRule=UseColor_Specified, specifiedColor=<rgba 1 1 1 0.7>}
    font {
     bForceMonospaced=false
     bMaterialIsStencil=false
     fontMaterial=null
     fontObject=&HeadingNow
     letterSpacing=10
     monospacedWidth=1
     size=15
     skewAmount=0
     typefaceFontName=74 Regular
     outlineSettings {
      bApplyOutlineToDropShadows=false
      bMiteredCorners=true
      bSeparateFillAlpha=false
      outlineMaterial=null
      outlineSize=0
      outlineColor=<rgba 0 0 0 0.4>
     }
    }
    shadowOffset=<0 0>
    strikeBrush {
     bIsDynamicallyLoaded=false
     drawAs=NoDrawType
     imageType=NoImage
     mirroring=NoMirror
     resourceName=None
     resourceObject=null
     tiling=NoTile
     imageSize=<32 32>
     margin=<ltrb 0 0 0 0>
     outlineSettings {
      bUseBrushTransparency=false
      roundingType=HalfHeightRadius
      width=0
      color {colorUseRule=UseColor_Specified, specifiedColor=<rgba 0 0 0 0>}
      cornerRadii=<0 0 0 0>
     }
     tintColor {colorUseRule=UseColor_Specified, specifiedColor=<rgba 1 1 1 1>}
     uVRegion {bIsValid=false, max=<0 0>, min=<0 0>}
    }
   }
   slot {class=BorderSlot, properties {verticalAlignment=VAlign_Top, padding=<ltrb 20 10 20 0>}}
  }
  OverlayModeToggle {
   widgetClassPath=&Overlay
   bIsVariable=true
   parent=OverlaySelectionUI
   widget=OverlayModeToggle
   slot {class=OverlaySlot, properties {horizontalAlignment=HAlign_Fill, verticalAlignment=VAlign_Fill}}
  }
  ModeSelectionBackground {
   widgetClassPath=&Image
   bIsVariable=true
   parent=OverlayModeToggle
   widget=ModeSelectionBackground
   properties {
    brush {
     bIsDynamicallyLoaded=false
     drawAs=Image
     imageType=FullColor
     mirroring=NoMirror
     resourceName=None
     resourceObject=/SparksCore/Widgets/Materials/MI_UI_BoomBox_Background
     tiling=NoTile
     imageSize=<32 32>
     margin=<ltrb 0 0 0 0>
     outlineSettings {
      bUseBrushTransparency=false
      roundingType=HalfHeightRadius
      width=0
      color {colorUseRule=UseColor_Specified, specifiedColor=<rgba 0 0 0 0>}
      cornerRadii=<0 0 0 0>
     }
     tintColor {colorUseRule=UseColor_Specified, specifiedColor=<rgba 1 1 1 1>}
     uVRegion {bIsValid=false, max=<0 0>, min=<0 0>}
    }
    colorAndOpacity=<rgba 1 1 1 0.4>
   }
   slot {class=OverlaySlot, properties {horizontalAlignment=HAlign_Fill, verticalAlignment=VAlign_Fill}}
  }
  ModeSelection_Item {
   widgetClassPath=&Image
   bIsVariable=true
   parent=OverlayModeToggle
   widget=ModeSelection_Item
   properties {brush {
     bIsDynamicallyLoaded=false
     drawAs=Image
     imageType=FullColor
     mirroring=NoMirror
     resourceName=None
     resourceObject=/SparksCore/Widgets/Materials/MI_UI_BoomBox_Button
     tiling=NoTile
     imageSize=<32 32>
     margin=<ltrb 0 0 0 0>
     outlineSettings {
      bUseBrushTransparency=false
      roundingType=HalfHeightRadius
      width=0
      color {colorUseRule=UseColor_Specified, specifiedColor=<rgba 0 0 0 0>}
      cornerRadii=<0 0 0 0>
     }
     tintColor {colorUseRule=UseColor_Specified, specifiedColor=<rgba 1 1 1 1>}
     uVRegion {bIsValid=false, max=<0 0>, min=<0 0>}
    }}
   slot {class=OverlaySlot, properties {horizontalAlignment=HAlign_Fill, verticalAlignment=VAlign_Fill, padding=<ltrb 6 6 6 6>}}
  }
  Image_JTP {
   widgetClassPath=&Image
   bIsVariable=true
   parent=OverlayModeToggle
   widget=Image_JTP
   properties {brush {
     bIsDynamicallyLoaded=false
     drawAs=Image
     imageType=FullColor
     mirroring=NoMirror
     resourceName=None
     resourceObject=/SparksCore/Widgets/Materials/MI_UI_BoomBox_Icon_Speaker
     tiling=NoTile
     imageSize=<36 36>
     margin=<ltrb 0 0 0 0>
     outlineSettings {
      bUseBrushTransparency=false
      roundingType=HalfHeightRadius
      width=0
      color {colorUseRule=UseColor_Specified, specifiedColor=<rgba 0 0 0 0>}
      cornerRadii=<0 0 0 0>
     }
     tintColor {colorUseRule=UseColor_Specified, specifiedColor=<rgba 1 1 1 1>}
     uVRegion {bIsValid=false, max=<0 0>, min=<0 0>}
    }}
   slot {class=OverlaySlot, properties {horizontalAlignment=HAlign_Center, verticalAlignment=VAlign_Center, padding=<ltrb 64 0 0 86>}}
  }
  Image_Jam {
   widgetClassPath=&Image
   bIsVariable=true
   parent=OverlayModeToggle
   widget=Image_Jam
   properties {brush {
     bIsDynamicallyLoaded=false
     drawAs=Image
     imageType=FullColor
     mirroring=NoMirror
     resourceName=None
     resourceObject=/SparksCore/Widgets/Materials/MI_UI_BoomBox_Icon_Instrument
     tiling=NoTile
     imageSize=<36 36>
     margin=<ltrb 0 0 0 0>
     outlineSettings {
      bUseBrushTransparency=false
      roundingType=HalfHeightRadius
      width=0
      color {colorUseRule=UseColor_Specified, specifiedColor=<rgba 0 0 0 0>}
      cornerRadii=<0 0 0 0>
     }
     tintColor {colorUseRule=UseColor_Specified, specifiedColor=<rgba 1 1 1 1>}
     uVRegion {bIsValid=false, max=<0 0>, min=<0 0>}
    }}
   slot {class=OverlaySlot, properties {horizontalAlignment=HAlign_Center, verticalAlignment=VAlign_Center, padding=<ltrb 0 0 64 86>}}
  }
  KeybindVisibilityControl {
   widgetClassPath=&GridPanel
   bIsVariable=true
   parent=OverlayModeToggle
   widget=KeybindVisibilityControl
   properties {columnFill [1], rowFill [1]}
   slot {class=OverlaySlot, properties {horizontalAlignment=HAlign_Fill, verticalAlignment=VAlign_Center, padding=<ltrb 0 10 0 102>}}
  }
  KeybindBackground {
   widgetClassPath=&Image
   bIsVariable=true
   parent=KeybindVisibilityControl
   widget=KeybindBackground
   properties {
    visibility=SelfHitTestInvisible
    brush {
     bIsDynamicallyLoaded=false
     drawAs=Image
     imageType=NoImage
     mirroring=NoMirror
     resourceName=None
     resourceObject=/FortUILibrary/Foundation/Art_ToolKit/Tools/M_UI_Circle_White
     tiling=NoTile
     imageSize=<48 48>
     margin=<ltrb 0 0 0 0>
     outlineSettings {
      bUseBrushTransparency=false
      roundingType=HalfHeightRadius
      width=0
      color {colorUseRule=UseColor_Specified, specifiedColor=<rgba 0 0 0 0>}
      cornerRadii=<0 0 0 0>
     }
     tintColor {colorUseRule=UseColor_Specified, specifiedColor=<rgba 1 1 1 1>}
     uVRegion {bIsValid=false, max=<0 0>, min=<0 0>}
    }
    colorAndOpacity=<rgba 0 0 0 0.4>
   }
   slot {class=GridSlot, properties {horizontalAlignment=HAlign_Center, verticalAlignment=VAlign_Center, padding=<ltrb 0 25 0 25>}}
  }
  Keybind_Holdable {
   widgetClassPath=/FortUILibrary/Systems/ModularButton/WBP_UIKit_Keybind_Holdable
   bIsVariable=true
   parent=KeybindVisibilityControl
   widget=Keybind_Holdable
   properties {bRequiresHold=false, triggeringInputAction {dataTable=/SparksCore/Input/SparksSongsEmoteWheelBindings, rowName=SwapSongOverrideMode}}
   slot {class=GridSlot}
  }
  TouchOverlayButton {
   widgetClassPath=&WBP_UIKit_Button_Generic
   bIsVariable=true
   parent=OverlayModeToggle
   widget=TouchOverlayButton
   properties {
    renderOpacity=0
    style=/FortUILibrary/Foundation/Buttons/ButtonStyles/ButtonStyle-Empty.ButtonStyle-Empty_C
    useIconBlock=false
    useTextBlock=false
    visibility=Collapsed
   }
   slot {class=OverlaySlot, properties {horizontalAlignment=HAlign_Fill, verticalAlignment=VAlign_Fill}}
  }
  ModeSelectionList {
   widgetClassPath=&CommonListView
   bIsVariable=true
   parent=OverlayModeToggle
   widget=ModeSelectionList
   properties {
    allowOverscroll=false
    bAllowDragging=false
    bEnableRightClickScrolling=false
    bEnableScrollAnimation=false
    bEnableTouchScrolling=false
    entryWidgetClass=&/WBP_SparksEmoteWheelModeButton.WBP_SparksEmoteWheelModeButton_C
    horizontalEntrySpacing=10
    numDesignerPreviewEntries=2
    orientation=Orient_Horizontal
    visibility=Hidden
   }
   slot {class=OverlaySlot, properties {horizontalAlignment=HAlign_Fill}}
  }
 ]
}
